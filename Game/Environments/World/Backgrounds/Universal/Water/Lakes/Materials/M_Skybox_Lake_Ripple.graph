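# Lit material for skybox background lakes, with animated ripples and time-of-day distance fog.
# A sphere mask on world XY blends color and roughness and fades in emissive fog, tinted between fog and lit colors by a second mask, all driven by MPC_Skybox_World.
# Ripples blend two oppositely panning, world-projected samples of a reused tree-canopy normal map, with tangent-space normals disabled.
# Inverted vertex blue scales a painted vertical height.
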Material /Game/Environments/World/Backgrounds/Universal/Water/Lakes/Materials/M_Skybox_Lake_Ripple
graph {
 properties {bTangentSpaceNormal=false, pixelDepthOffsetMode=PDOM_Legacy, usedShadingModels=Default Lit}
 nodes [
  Constant3Vector() {constant=<rgba 0.0131 0.0502 0.07 1>}
  Constant()
  n1 = ComponentMask(None=WorldPosition() {worldPositionShaderOffset=WPT_ExcludeAllShaderOffsets}.XYZ) {g=true, r=true}
  n2 = CollectionParameter() {
   collection=/Game/Environments/World/Backgrounds/Universal/MasterMaterials/MPC_Skybox_World
   parameterName="Morn_R=Distance_G=DiffuseAmt_B=EmissiveAmt"
  }
  n3 = CollectionParameter() {
   collection=&MPC_Skybox_World
   parameterName="Day_R=Distance_G=DiffuseAmt_B=EmissiveAmt"
  }
  n4 = CollectionParameter() {
   collection=&MPC_Skybox_World
   parameterName="Eve_R=Distance_G=DiffuseAmt_B=EmissiveAmt"
  }
  n5 = CollectionParameter() {
   collection=&MPC_Skybox_World
   parameterName="Night_R=Distance_G=DiffuseAmt_B=EmissiveAmt"
  }
  n6 = MaterialFunctionCall(Float3=MaterialFunctionCall(Color_Morning=@n2, Color_Day=@n3, Color_Evening=@n4, Color_Night=@n5) {materialFunction=&Color_Time_of_Day}.Result) {materialFunction=&BreakOutFloat3Components}
  n7 = SphereMask(A=@n1, B=Constant2Vector(), Radius=@n6.R, Hardness=Constant())
  LinearInterpolate(A=ScalarParameter() {defaultValue=0.7, parameterName=RoughnessA}, B=ScalarParameter() {defaultValue=0.3, parameterName=Roughness}, Alpha=@n7)
  Multiply(A=VectorParameter() {parameterName=Color, defaultValue=<rgba 0.0152 0.07 0.0373 1>}.RGB, B=LinearInterpolate(A=@n7, Alpha=@n6.G)) {desc=BaseColor}
  n8 = MaterialFunctionCall(
   Color_Morning=CollectionParameter() {collection=&MPC_Skybox_World, parameterName=MorningColor}
   Color_Day=CollectionParameter() {collection=&MPC_Skybox_World, parameterName=DayColor}
   Color_Evening=CollectionParameter() {collection=&MPC_Skybox_World, parameterName=EveningColor}
   Color_Night=CollectionParameter() {collection=&MPC_Skybox_World, parameterName=NightColor}
  ) {materialFunction=&Color_Time_of_Day}
  n9 = MaterialFunctionCall(
   Color_Morning=CollectionParameter() {collection=&MPC_Skybox_World, parameterName=MorningLitColor}
   Color_Day=CollectionParameter() {collection=&MPC_Skybox_World, parameterName=DayLitColor}
   Color_Evening=CollectionParameter() {collection=&MPC_Skybox_World, parameterName=EveningLitColor}
   Color_Night=CollectionParameter() {collection=&MPC_Skybox_World, parameterName=NightLitColor}
  ) {materialFunction=&Color_Time_of_Day}
  n10 = CollectionParameter() {
   collection=&MPC_Skybox_World
   parameterName="Morn_R=LitLength_G=LitRot_B=LitOffset"
  }
  n11 = CollectionParameter() {
   collection=&MPC_Skybox_World
   parameterName="Day_R=LitLength_G=LitRot_B=LitOffset"
  }
  n12 = CollectionParameter() {
   collection=&MPC_Skybox_World
   parameterName="Eve_R=LitLength_G=LitRot_B=LitOffset"
  }
  n13 = CollectionParameter() {
   collection=&MPC_Skybox_World
   parameterName="Night_R=LitLength_G=LitRot_B=LitOffset"
  }
  n14 = MaterialFunctionCall(Float3=MaterialFunctionCall(Color_Morning=@n10, Color_Day=@n11, Color_Evening=@n12, Color_Night=@n13) {materialFunction=&Color_Time_of_Day}.Result) {materialFunction=&BreakOutFloat3Components}
  n15 = Subtract(A=Add(A=@n1, B=Constant() {r=4000000}), B=AppendVector(A=@n14.G, B=@n14.G))
  n16 = Constant2Vector()
  n17 = Constant()
  Multiply(
   A=Multiply(A=@n6.B, B=OneMinus(None=@n7))
   B=LinearInterpolate(
    A=ComponentMask(None=@n8.Result) {b=true, g=true, r=true}
    B=ComponentMask(None=@n9.Result) {b=true, g=true, r=true}
    Alpha=SphereMask(A=@n15, B=@n16, Radius=@n14.R, Hardness=@n17)
   )
  ) {desc=Emissive}
  Multiply(
   A=OneMinus(None=VertexColor().B)
   B=Multiply(A=Constant3Vector() {constant=<rgba 0 0 1 0>}, B=ScalarParameter() {defaultValue=4096, parameterName=Painted Height})
  )
  n18 = ComponentMask(None=WorldPosition() {worldPositionShaderOffset=WPT_ExcludeAllShaderOffsets}.XYZ) {g=true, r=true}
  n19 = Multiply(A=Time(), B=Constant() {r=0.02})
  n20 = Panner(Coordinate=Divide(A=@n18, B=ScalarParameter() {defaultValue=50000, parameterName=TileA}) {constB=1024, desc=Normalmap Scale}, Time=@n19) {speedX=-0.5, speedY=2}
  n21 = Panner(Coordinate=Divide(A=@n18, B=ScalarParameter() {defaultValue=80000, parameterName=TileB}) {constB=1024, desc=Normalmap Scale}, Time=@n19) {speedX=0.3, speedY=-1}
  Multiply(
   A=MaterialFunctionCall(
    BaseNormal=TextureSample(UVs=@n20) {samplerType=SAMPLERTYPE_Normal, texture=/Game/Environments/World/Backgrounds/Locales/Temperate/Textures/T_Skybox_TreeCanopy_01_N}.RGB
    AdditionalNormal=TextureSample(UVs=@n21) {samplerType=SAMPLERTYPE_Normal, texture=&T_Skybox_TreeCanopy_01_N}.RGB
   ) {materialFunction=&BlendAngleCorrectedNormals}.Result
   B=VectorParameter() {parameterName=NormalModifier, defaultValue=<rgba -0.3 -0.3 1 1>}.RGB
  )
  ScalarParameter() {defaultValue=800000, parameterName=SkyboxFogDistance_1}
  ScalarParameter() {defaultValue=0.1, parameterName=SkyboxFogDiffuseAmt_1}
  ScalarParameter() {defaultValue=0.1, parameterName=SkyboxFogEmissiveAmt_1}
  ScalarParameter() {defaultValue=4100000, parameterName=SkyboxFogLightLength_1}
  ScalarParameter() {defaultValue=0.16, parameterName=SkyboxFogLightRotation_1}
  ScalarParameter() {defaultValue=3, parameterName=SkyboxFogLittOffset_Debug}
  ScalarParameter() {defaultValue=1, parameterName=Specular}
  VectorParameter() {parameterName=SkyboxFogLightColor, defaultValue=<rgba 0.3049 0.5404 1.5 1>}
  VectorParameter() {parameterName=SkyboxFogColor, defaultValue=<rgba 0.2484 0.2991 0.92 1>}
 ]
}
